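# Animation Blueprint for the HerbalFist backpack cosmetic, adding secondary physics motion to its idle animation.
# The idle sequence feeds a root-relative AnimDynamics node that swings the dyn_paper_01 bone with an angular spring, tight limits and reduced gravity.
# An additive bone-space Modify Bone then slightly offsets and rolls the pack bone.
# The update event converts bIsSkyDiving into a numeric Skydiving variable, which the AnimGraph does not read.
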AnimBlueprint /BRCosmetics/Accessories/FORT_Backpacks/Backpack_M_MED_HerbalFist/Meshes/M_MED_HerbalFist_Pack_AnimBP
target_skeleton=&/M_MED_HerbalFist_Pack_Skeleton
graphs [
 AnimGraph {nodes [
   n1 = Anim Dynamics - Bone: dyn_paper_01(
    ComponentPose=Local To Component(LocalPose=Sequence Player 'HerbalFist_IdleAnim_PACK'() {sequence=/BRCosmetics/Accessories/FORT_Backpacks/Backpack_M_MED_HerbalFist/Anim/HerbalFist_IdleAnim_PACK})
    Alpha=1
    ExternalForce="0,0,0"
   ) {
    angularDampingOverride=1
    angularSpringConstant=10000
    bAngularSpring=true
    bOverrideAngularDamping=true
    bUseGravityOverride=true
    gravityScale=0.35
    simulationSpace=RootRelative
    boundBone {boneName=dyn_paper_01}
    componentAppliedLinearAccClamp=<2000 2000 2000>
    componentLinearAccScale=<0.5 0.5 0.5>
    componentLinearVelScale=<0.5 0.5 0.5>
    gravityOverride {z=-980}
    physicsBodyDefinitions [{
      boundBone {boneName=dyn_paper_01}
      boxExtents=<20 20 20>
      constraintSetup {angularLimitsMax=<2 10 10>, angularLimitsMin=<-2 -10 -10>, angularTarget=<1 0 0>}
      localJointOffset=<1 0 0>
     }]
    simSpaceSettings {simSpaceAngularAlpha=0.6}
   }
   n2 = "Transform (Modify) Bone - Bone: pack"(ComponentPose=@n1, Translation="0,-0.5,0", Rotation="0,0,5") {
    rotationMode=BMM_Additive
    rotationSpace=BCS_BoneSpace
    translationMode=BMM_Additive
    translationSpace=BCS_BoneSpace
    boneToModify {boneName=pack}
    rotation {roll=5}
    translation {y=-0.5}
   }
   Output Pose(Result=Component To Local(ComponentPose=@n2)) {layerGroup=None}
  ]}
 EventGraph {nodes [
   Event BlueprintUpdateAnimation()
   > Set Skydiving(Skydiving:double="To Float (Boolean)"(InBool=Get bIsSkyDiving())) -> (Output_Get:double)
  ]}
]
variables [Skydiving {type=number}]
